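; Interactable weed for Reality Saplings that players pick, notifying its sapling and dying with dirt, dust and sound effects.
; A looping timer interpolates mesh scale for randomly staggered grow-in and scale-out death.
; Server registers it with the sapling and binds its death to the supporting floor actor's.
; Replicated owner and playlist-driven all-players state, pushed via net dormancy flushes, decide local visibility and interactability.
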(event EventBeginPlay
  (Networking|SwitchHasAuthority
    (:Authority
      (Variables|Default|SetbAvailableLocally true)
      (Class|FortAsyncActionCurrentPlaylistReady|PlaylistDataReadyAsync
        (:then
          _)
        (:OnReady
          (Variables|Default|SetPlaylistDataReady true)
          (CallFunction|InitializeAllPlayersCanInteractAndSeeState))))
    (:Remote
      _))
  (Rendering|SetVisibility (Variables|Default|GetStaticMesh) false false)
  (bind _output_get (Variables|Default|SetCur%20Min%20World%20Scale (Variables|Default|GetStart%20World%20Scale)))
  (Default|SetWeedScale _output_get)
  (Variables|Default|SetCur%20Max%20World%20Scale (Variables|Default|GetDest%20World%20Scale))
  (Variables|Default|SetHasReachedDestScale false)
  (Utilities|FlowControl|Delay (Math|Random|RandomFloatinRange 0.0 0.45))
  (Variables|Default|SetpastFirstDelay true)
  (Rendering|SetVisibility (Variables|Default|GetStaticMesh) (Variables|Default|GetbAvailableLocally) false)
  (bind _returnvalue (Utilities|Time|SetTimerbyEvent (delegate Custom|UpdateLerpToDest) (Variables|Default|GetLerpTimerInterval) true))
  (Variables|Default|SetLerpTimerHandle _returnvalue))

(event Custom|UpdateLerpToDest
  (Utilities|FlowControl|Sequence
    (:then_0
      ;[1 1 1] Lerp Scale as needed
      (if (not (Variables|Default|GetHasReachedDestScale))
        (Variables|Default|SetCurr%20World%20Scale (Math|Interpolation|FInterpTo (Variables|Default|GetCurr%20World%20Scale) (Variables|Default|GetCur%20Max%20World%20Scale) (Variables|Default|GetLerpTimerInterval) (Variables|Default|GetScale%20Interp%20Speed)))
        (if (<= (Math|Float|Absolute(Float) (- (Variables|Default|GetCurr%20World%20Scale) (Variables|Default|GetCur%20Max%20World%20Scale))) 0.02)
          (Default|SetWeedScale (Variables|Default|GetCur%20Max%20World%20Scale))
          (Variables|Default|SetHasReachedDestScale true)
          (if (not (Variables|Default|GetScaleOutAndDie))
            (Variables|/Script/Engine.PrimitiveComponent|SetbGenerateOverlapEvents true (Variables|Default|GetStaticMesh)))
          (else
            (Default|SetWeedScale (Variables|Default|GetCurr%20World%20Scale))))))
    (:then_1
      ;[1 1 1] Cancel looping once we've hit our target
      (if (Variables|Default|GetHasReachedDestScale)
        (Utilities|Time|ClearandInvalidateTimerbyHandle (Variables|Default|GetLerpTimerHandle))
        (if (Variables|Default|GetScaleOutAndDie)
          (Networking|SwitchHasAuthority
            (:Authority
              (Utilities|FlowControl|Delay 1.0)
              (Gameplay|SilentDie false))
            (:Remote
              (Rendering|SetVisibility (Variables|Default|GetStaticMesh) false true))))))))

(event Custom|Scale%20Out%20and%20Die
  ;[1 1 1] Scale Out And Die
  (Variables|Default|SetScaleOutAndDie true)
  (Utilities|FlowControl|Delay (Math|Random|RandomFloatinRange 0.0 0.4))
  (Variables|Default|SetCur%20Min%20World%20Scale (Variables|Default|GetCurr%20World%20Scale))
  (Variables|Default|SetCur%20Max%20World%20Scale (Variables|Default|GetStart%20World%20Scale))
  (Variables|Default|SetHasReachedDestScale false)
  (Variables|/Script/Engine.PrimitiveComponent|SetbGenerateOverlapEvents false (Variables|Default|GetStaticMesh))
  (bind _returnvalue (Utilities|Time|SetTimerbyEvent (delegate Custom|UpdateLerpToDest) (Variables|Default|GetLerpTimerInterval) true))
  (Variables|Default|SetLerpTimerHandle _returnvalue))

(event Custom|Base%20Died
  (params (DamagedActor :type Actor%20Object%20Reference) (Damage :type Float%20%28single-precision%29) (InstigatedBy :type Controller%20Object%20Reference) (DamageCauser :type Actor%20Object%20Reference) (HitLocation :type Vector) (FHitComponent :type Primitive%20Component%20Object%20Reference) (BoneName :type Name) (Momentum :type Vector))
  ;[1 1 1] Bind To Supporting Actor Destroyed
  (Gameplay|Die))

(event Custom|Bind%20To%20Supporting%20Actor%20Destroyed
  (bind (_floorhit_bblockinghit _floorhit_binitialoverlap _floorhit_time _floorhit_distance _floorhit_location _floorhit_impactpoint _floorhit_normal _floorhit_impactnormal _floorhit_physmat _floorhit_hitactor) (Collision|BreakHitResult (Variables|GetFloorHit_bBlockingHit)))
  ;[1 1 1] Bind To Supporting Actor Destroyed
  (bind _asbuilding_actor (Utilities|Casting|CastToBuildingActor _floorhit_hitactor)
    (:then
      (EventDispatchers|BindEventtoOnDied _asbuilding_actor (delegate Custom|Base%20Died)))
    (:CastFailed
      (Game|BindEventtoOnDestroyed _floorhit_hitactor (delegate Custom|Base%20Destroyed)))))

(event Custom|Base%20Destroyed
  (params (DestroyedActor :type Actor%20Object%20Reference))
  ;[1 1 1] Bind To Supporting Actor Destroyed
  (Gameplay|Die))

(event Interaction|EventBlueprintOnInteract
  (params (InteractingPawn :type Fort%20Pawn%20Object%20Reference) (InteractionBeingAttempted :type EInteractionBeingAttempted%20Enum :default "FirstInteraction"))
  (|CallParentFunction InteractingPawn InteractionBeingAttempted)
  (Utilities|FlowControl|Sequence
    (:then_0
      (Default|CallSaplingWeededbyEnemyEvent (Variables|Default|GetSapling)))
    (:then_1
      (CallFunction|DestroyWeed InteractingPawn))))

(event EventDestroyed
  (CallFunction|DestroyWeed 0))

(event Custom|DestroyWeed
  (params (NewParam :type Fort%20Pawn%20Object%20Reference))
  (bind _sapling (Variables|Default|GetSapling))
  (Utilities|IsValid _sapling
    (:Is%20Valid
      (Class|BSMAAthenaRealitySaplingParent|OnWeedDestroyed _sapling self))
    (:Is%20Not%20Valid
      _))
  (Gameplay|Die "(TagName=\"Gameplay.Effect.InstantDeath.Weeded\")" 0 NewParam))

(event EventOnDeathPlayEffects
  (params (Damage :type Float%20%28single-precision%29) (DamageTags :type Gameplay%20Tag%20Container%20Structure%20%28by%20ref%29) (Momentum :type Vector :default "0, 0, 0") (HitInfo :type Hit%20Result%20Structure%20%28by%20ref%29) (InstigatedBy :type Fort%20Pawn%20Object%20Reference) (DamageCauser :type Actor%20Object%20Reference) (EffectContext :type Gameplay%20Effect%20Context%20Handle%20Structure))
  (Niagara|SpawnSystematLocation "/RealitySeedGameplay/Environment/Assets_RealitySapling/FX/NS_WeedingDirt.NS_WeedingDirt" (Transformation|GetActorLocation))
  (Audio|PlaySoundatLocation (Variables|Default|GetWeedPulledSound) (Transformation|GetActorLocation))
  (Effects|SpawnEmitteratLocation :EmitterTemplate "/Game/Athena/Items/Consumables/DurrrablesBurger/P_ConsumePuffDust.P_ConsumePuffDust" :Location (Transformation|GetActorLocation) :Scale "1.000000,1.000000,1.000000" :bAutoDestroy true :PoolingMethod "AutoRelease" :bAutoActivateSystem true)
  (Rendering|SetActorHiddenInGame true))

(fn AddWeedToSapling
  (Networking|SwitchHasAuthority
    (:Authority
      (Class|BSMAAthenaRealitySaplingParent|AddServerWeed (Variables|Default|GetSapling) self)
      (CallFunction|BindtoSupportingActorDestroyed))))

(fn BlueprintCanInteract
  (params (InteractingPawn :type Fort%20Pawn%20Object%20Reference) (InteractionBeingAttempted :type EInteractionBeingAttempted%20Enum :default "FirstInteraction") (InteractionType :type TInteractionType%20Enum :default "IT_NoInteraction"))
  (if (Variables|Default|GetbAvailableLocally)
    (return (Utilities|Enum|Equal(Enum) InteractionBeingAttempted "FirstInteraction"))))

(fn BlueprintGetInteractionString
  (params (InteractingPawn :type Fort%20Pawn%20Object%20Reference) (InteractionBeingAttempted :type EInteractionBeingAttempted%20Enum :default "FirstInteraction"))
  (return (Utilities|Select :Index InteractionBeingAttempted :FirstInteraction "NSLOCTEXT(\"[0827087A482181CDB3E128AA75EFEE61]\", \"62C62EBD4D724E180C009C935F2D0C40\", \"Pick Weed\")")))

(fn Destroy)

(fn InitializeAllPlayersCanInteractandSeeState
  (Networking|SwitchHasAuthority
    (:Authority
      (if (Variables|Default|GetPlaylistDataReady)
        (if (ScalableFloat|GetValueasBool (Variables|Default|GetEnabledForAllPlayers))
          (Variables|Default|SetAllPlayersCanSeeAndInteractState 1)
          (else
            (Variables|Default|SetAllPlayersCanSeeAndInteractState 0)))
        (Networking|FlushNetDormancy) ; Weed starts dormant. Force a replication for AllPlayersCanSeeAndInteractState
      ))))

(fn InitializeFromLocalPlayerControllerAndDetermineEnabledforLocalPlayer
  (bind _owning_player_controller_repped (Variables|Default|GetOwning%20Player%20Controller%20Repped))
  (bind _returnvalue (Game|GetLocalPlayerControllerID _owning_player_controller_repped))
  (if (== (Variables|Default|GetAllPlayersCanSeeAndInteractState) 1)
    (Variables|Default|SetbAvailableLocally true)
    (if (Variables|Default|GetpastFirstDelay)
      (Rendering|SetVisibility (Variables|Default|GetStaticMesh) (Variables|Default|GetbAvailableLocally) false))
    (else
      ;[1 1 1] If this isn't owned by the local player hide it and no interacting with it.
      (Utilities|IsValid (Variables|Default|GetOwning%20Player%20Controller%20Repped)
        (:Is%20Valid
          (Development|PrintString (Utilities|String|Append "Owning player controller repped is: " (Utilities|GetDisplayName _owning_player_controller_repped) "     LocalPlayerController ID:" (Utilities|String|ToString(Integer) _returnvalue)))
          (if (!= _returnvalue -1)
            (Variables|Default|SetbAvailableLocally true)
            (if (Variables|Default|GetpastFirstDelay)
              (Rendering|SetVisibility (Variables|Default|GetStaticMesh) (Variables|Default|GetbAvailableLocally) false))))))))

(fn OnInteractAllowedToConsume
  (params (InteractingPawn :type Fort%20Pawn%20Object%20Reference) (InteractionBeingAttempted :type EInteractionBeingAttempted%20Enum :default "FirstInteraction"))
  ;[1 1 1] If second interact allow gather if enabled
  (switch Utilities|FlowControl|Switch|SwitchonEInteractionBeingAttempted InteractionBeingAttempted
    (:FirstInteraction
      (return true))))

(fn OnRep_AllPlayersCanSeeAndInteractState
  (CallFunction|InitializefromLocalPlayerControllerandDetermineEnabledForLocalPlayer))

(fn OnRep_OwningPlayerControllerRepped
  (if (!= (Variables|Default|GetAllPlayersCanSeeAndInteractState) -1)
    (CallFunction|InitializeAllPlayersCanInteractAndSeeState)))

(fn OnRep_Sapling
  (bind _sapling (Variables|Default|GetSapling))
  (bind _owning_player_controller_repped (Variables|/RealitySeedGameplay/items/RealitySaplings/B_SMA_Athena_RealitySapling_Parent.B_SMA_Athena_RealitySapling_Parent_C|GetOwning%20Player%20Controller%20Repped _sapling))
  (Utilities|IsValid _sapling
    (:Is%20Valid
      (Utilities|IsValid _owning_player_controller_repped
        (:Is%20Valid
          (Variables|Default|SetOwning%20Player%20Controller%20Repped _owning_player_controller_repped)
          (Networking|SwitchHasAuthority
            (:Authority
              (Networking|FlushNetDormancy) ; Weed starts dormant. Force a replication for Sapling and OwningPlayerControllerRepped
            )))))))

(fn SetWeedScale
  (params (SizeAlpha :type Float%20%28double-precision%29))
  (bind _output_get (Variables|Default|SetCurr%20World%20Scale SizeAlpha))
  (Transformation|SetWorldScale3D (Variables|Default|GetStaticMesh) (Math|Vector|MakeVector _output_get _output_get _output_get)))

(fn ShouldDie
  (params (Damage :type Float%20%28single-precision%29) (EventInstigator :type Controller%20Object%20Reference) (DamageCauser :type Actor%20Object%20Reference) (bIsTest :type Boolean) (EffectContext :type Gameplay%20Effect%20Context%20Handle%20Structure%20%28by%20ref%29))
  (return false))

(fn ConstructionScript
  (bind (_returnvalue_roll _returnvalue_pitch) (Math|Rotator|BreakRotator (Transformation|GetWorldRotation (Variables|Default|GetStaticMesh))))
  (|CallParentFunction)
  (Variables|Default|SetDest%20World%20Scale (Math|Random|RandomFloatinRange 0.6 0.75))
  (Transformation|SetWorldScale3D (Variables|Default|GetStaticMesh) (Math|Conversions|ToVector(Float) (Variables|Default|GetDest%20World%20Scale)))
  (Transformation|SetWorldRotation (Variables|Default|GetStaticMesh) (Math|Rotator|MakeRotator _returnvalue_roll _returnvalue_pitch (Math|Random|RandomFloatinRange 0.0 3.6e+02)) false false))
