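; Spawner that places weeds around a Reality Sapling, using precomputed EQS locations.
; After a spawn-chance roll, it line-traces at each chosen location, caches the hit, and spawns with random scale, optional surface-normal alignment and random yaw, then destroys itself.
; On each server spawn, it links the weed to its sapling and, with authority, passes it the cached floor hit and registers it with the sapling.
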(event EventBeginPlay
  (|CallParentFunction) ; This node is disabled and will not be called.\nDrag off pins to build functionality.
)

(event Collision|EventActorBeginOverlap
  (params (OtherActor :type Actor%20Object%20Reference)))

(event EventTick
  (params (DeltaSeconds :type Float%20%28single-precision%29)))

(event EventServerActorSpawned
  (params (Actor :type Actor%20Object%20Reference))
  (bind _asbga_weed_simple (Utilities|Casting|CastToBGA_Weed_Simple Actor))
  (Variables|/RealitySeedGameplay/Environment/Foliage/BGA_Weed_Simple.BGA_Weed_Simple_C|SetSapling (Variables|Default|GetSapling) _asbga_weed_simple)
  (Networking|SwitchHasAuthority
    (:Authority
      (Variables|/RealitySeedGameplay/Environment/Foliage/BGA_Weed_Simple.BGA_Weed_Simple_C|SetFloorHit (Variables|Default|GetLastHit) _asbga_weed_simple)
      (Default|AddWeedtoSapling _asbga_weed_simple))))

(event EventDestroyed)

(event EventStartEQSSpawnLoop
  (bind _locationstospawn (Variables|Non-Tunables|GetLocationsToSpawn))
  (bind _spawnlocationoffset (Variables|SpawnerVars|GetSpawnLocationOffset))
  ;[1 1 1] StartEQSSpawn - Loop through and spawn the fishing pools. Run a test to see if we should spawn this one. If we fail, still increment but don't spawn
  ;;[1 1 1] Determine whether this Cluster will spawn at all based on Spawn Chance
  (if (Math|Random|RandomBoolwithWeight (Math|Float|Clamp(Float) (ScalableFloat|GetValueatLevel (Variables|SpawnerVars|GetRow_SpawnChance))))
    ;;[1 1 1] Calcualte the # of actors to spawn at this cluster
    (bind _output_get (Variables|Non-Tunables|SetcachedSpawnAmount (Math|Integer|Clamp(Integer) (CallFunction|GetSpawnAmountforEQS) 0 (Utilities|Array|Length (Variables|Non-Tunables|GetLocationsToSpawn)))))
    (for _index (range _output_get)
      (bind (_outhit _returnvalue) (Collision|LineTraceByChannel (+ (+ (Utilities|Array|Get(aref) _locationstospawn _index) _spawnlocationoffset) "0.000000,0.000000,50.000000") (+ (+ (Utilities|Array|Get(aref) _locationstospawn _index) _spawnlocationoffset) "0.000000,0.000000,-50.000000")))
      (Variables|Default|SetLastHit _outhit)
      (CallFunction|SpawnActor (Math|Transform|MakeTransform (+ (Utilities|Array|Get(aref) _locationstospawn _index) _spawnlocationoffset) (Math|Rotator|SelectRotator (Math|Rotator|MakeRotfromZX (.Normal (Collision|BreakHitResult _outhit)) (Math|Vector|SelectVector (Math|Vector|RotateVectorAroundAxis "1.000000,0.000000,0.000000" (Math|Random|RandomFloatinRange -1.8e+02 1.8e+02) (.Normal (Collision|BreakHitResult _outhit))) (Math|Vector|RotateVectorAroundAxis "1.000000,0.000000,0.000000" 9e+01 (.Normal (Collision|BreakHitResult _outhit))) (Variables|SpawnerVars|GetEQS_RandomYaw))) (Math|Rotator|SelectRotator (Math|Rotator|MakeRotfromZX (Math|Vector|RotateVectorAroundAxis "0.000000,0.000000,1.000000" 0.0 (Math|Rotator|GetRotationXVector (Transformation|GetActorRotation))) (Math|Rotator|GetRotationXVector (Transformation|GetActorRotation))) (Transformation|GetActorRotation) (Variables|SpawnerVars|GetEQS_RandomYaw)) (and _returnvalue (Variables|SpawnerVars|GetEQS_AlignToSurfaceNormal))) (* "1.000000,1.000000,1.000000" (Math|Random|RandomFloatinRange (Variables|SpawnerVars|GetEQS_ScaleMIN) (Variables|SpawnerVars|GetEQS_ScaleMAX))))))
    (CallFunction|DestroySpawner)))

(fn ConstructionScript
  (|CallParentFunction))
